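# Parent material for the DemonEyes vehicle wheel cosmetic, adding an animated emissive effect to the simple wheel parent function.
# A slowly panning waterfall texture is sharpened and weighted by red and green channels of a VFX mask, with a green-masked layer added.
# A Fresnel term adds further glow, and named reroutes expose base color, normal and a custom color parameter.
# The resulting attributes pass through the shared vehicle damage and de-rez transition functions.
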Material /VehicleCosmetics/Wheels/DemonEyes/Materials/M_WheelParent_DemonEyes
graph {
 properties {
  bAutomaticallySetUsageInEditor=false
  bUseMaterialAttributes=true
  bUsedWithMeshDeformer=true
  bUsedWithSkeletalMesh=true
  pixelDepthOffsetMode=PDOM_Legacy
  usedShadingModels=Default Lit
 }
 nodes [
  n1 = GetMaterialAttributes(None=MaterialFunctionCall() {materialFunction=&MF_WheelParent_Simple}.MaterialAttributes) {attributeGetTypes [#1, #2]}
  n2 = TextureSample(UVs=Panner(Coordinate=TextureCoordinate() {uTiling=2.5, vTiling=4}) {speedX=0.1}) {texture=/VehicleCosmetics/Wheels/_SharedTextures/FX/Waterfall_01_H}
  n3 = TextureSampleParameter2D() {parameterName=Masks VFX, samplerType=SAMPLERTYPE_LinearColor, texture=/VehicleCosmetics/Wheels/DemonEyes/Textures/T_DemonEyes_VFX_M}
  n4 = Multiply(A=Multiply(A=Power(Base=@n2.RGB), B=NamedRerouteUsage()), B=Add(A=Multiply(A=@n3.R) {constB=4}, B=@n3.G))
  n5 = Multiply(A=NamedRerouteUsage(), B=Multiply(A=@n3.G) {constB=0.5})
  n6 = NamedRerouteUsage()
  n7 = Constant() {r=3}
  n8 = Add(
   A=Multiply(A=Add(A=@n4, B=@n5)) {constB=3}
   B=Multiply(
    A=NamedRerouteUsage()
    B=MaterialFunctionCall!(Normal=@n6, Exponent=@n7) {materialFunction=/VehicleCosmetics/Wheels/Cola_Holographic/Materials/MF_Fresnel_Classic}.Result
   )
  )
  MaterialFunctionCall(In=MaterialFunctionCall(Material=SetMaterialAttributes(MaterialAttributes=@n1.MaterialAttributes, Emissive Color=@n8) {attributeSetTypes [#3]}) {materialFunction=&MF_BaseVehicle_Damage}.Result) {materialFunction=&MF_VehicleDeRez_Transition}
  NamedRerouteDeclaration(None=@n1.Base Color) {name=BC, nodeColor=<rgba 0.4766 1 0 1>}
  NamedRerouteDeclaration(None=@n1.Normal) {name=Normal, nodeColor=<rgba 1 0 0.7969 1>}
  NamedRerouteDeclaration(None=VectorParameter() {parameterName=Custom Color, defaultValue=<rgba 1 0.0513 0 1>}.RGB) {name=Custom Color, nodeColor=<rgba 0 1 0.8828 1>}
 ]
}
