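# Unlit material for the Demon Eyes wheel cosmetic, drawing an animated emissive portal with an eye, tinted by a Custom Color parameter.
# Two panning gradient samples build a jitter offset, added to a strongly negative BumpOffset to form parallax eye UVs.
# Radial-coordinate textures drive scrolling noise-distorted clouds and sphere-masked dot particles.
# These layers and the eye are blended by lerps, with a parameter balancing particles against the eye, into Emissive.
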Material /VehicleCosmetics/Wheels/DemonEyes/Materials/M_Wheel_DemonEyes_Portal1
graph {
 properties {
  bAutomaticallySetUsageInEditor=false
  bUseMaterialAttributes=true
  bUsedWithMeshDeformer=true
  bUsedWithSkeletalMesh=true
  shadingModel=MSM_Unlit
  usedShadingModels=Unlit
 }
 nodes [
  n1 = TextureCoordinate() {uTiling=0, vTiling=0}
  n2 = Multiply(A=ScalarParameter() {defaultValue=3, parameterName=JitterAmount}, B=Time())
  n3 = TextureSample(UVs=Panner(Coordinate=@n1, Time=@n2) {speedX=-0.1, speedY=0.2}) {samplerSource=SSM_Wrap_WorldGroupSettings, texture=/VehicleCosmetics/Wheels/_SharedTextures/FX/MowGradient01}
  n4 = TextureSample(UVs=Panner(Coordinate=@n1, Time=@n2) {speedX=0.12, speedY=-0.05}) {samplerSource=SSM_Wrap_WorldGroupSettings, texture=&MowGradient01}
  n5 = Clamp(None=Multiply(A=AppendVector(A=@n3.G, B=@n4.G)) {constB=0.3})
  n6 = Add(A=TextureCoordinate() {uTiling=2, vTiling=2}) {constB=-0.5}
  n7 = Constant() {r=-10}
  NamedRerouteDeclaration(None=Add(A=ConstantBiasScale(None=@n5) {bias=-0.5, scale=0.15}, B=BumpOffset(Coordinate=@n6, Height=@n7))) {name=Jitter UV, nodeColor=<rgba 1 0 0.7734 1>}
  n8 = TextureSample() {
   samplerSource=SSM_Wrap_WorldGroupSettings
   samplerType=SAMPLERTYPE_LinearColor
   texture=/VehicleCosmetics/Wheels/_SharedTextures/FX/Radial_Coords01_D
  }
  n9 = Time()
  n10 = AppendVector(A=@n9, B=@n9)
  n11 = TextureSample(UVs=Add(A=ComponentMask(None=@n8.RGB) {g=true, r=true}, B=Multiply(A=@n10) {constB=-0.1})) {samplerSource=SSM_Wrap_WorldGroupSettings, texture=/VehicleCosmetics/Wheels/_SharedTextures/FX/NoiseClouds}
  n12 = TextureSample(UVs=Add(A=ConstantBiasScale(None=@n11.G) {bias=-0.5, scale=0.15}, B=TextureCoordinate())) {texture=/VehicleCosmetics/Wheels/_SharedTextures/FX/Cloud_T}
  n13 = Multiply(A=@n11.RGB, B=@n12.RGB)
  NamedRerouteDeclaration(None=Clamp(None=Add(A=@n13, B=@n12.RGB))) {name=Radial Coord Cloud, nodeColor=<rgba 1 0 0.7734 1>}
  n14 = TextureSample() {
   samplerSource=SSM_Wrap_WorldGroupSettings
   samplerType=SAMPLERTYPE_LinearColor
   texture=&Radial_Coords01_D
  }
  n15 = Time()
  n16 = AppendVector(A=@n15, B=@n15)
  n17 = Constant2Vector() {g=-0.1, r=-0.2}
  n18 = TextureSample(UVs=Add(A=ComponentMask(None=@n14.RGB) {g=true, r=true}, B=Multiply(A=@n16, B=@n17) {constB=-0.1})) {samplerSource=SSM_Wrap_WorldGroupSettings, texture=/VehicleCosmetics/Wheels/_SharedTextures/FX/Dots_RGB_Med_T}
  n19 = Add(
   A=Multiply(A=@n18.R, B=ScalarParameter() {defaultValue=10, parameterName=Dots_Emis_01})
   B=Multiply(A=@n18.G, B=ScalarParameter() {defaultValue=1, parameterName=Dots_Emis_02})
  )
  n20 = SphereMask(A=TextureCoordinate(), B=Constant2Vector() {g=0.5, r=0.5}, Radius=Constant() {r=0.3}, Hardness=Constant() {r=0.5})
  NamedRerouteDeclaration(None=Multiply(A=LinearInterpolate(A=@n19, Alpha=@n20) {constB=0}, B=ScalarParameter() {defaultValue=3, parameterName=Particle_Emis})) {name=Radial Coord Particles, nodeColor=<rgba 1 0 0.7734 1>}
  n21 = NamedRerouteUsage()
  n22 = VectorParameter() {parameterName=Custom Color, defaultValue=<rgba 1 0.0513 0 1>}
  n23 = Multiply(A=Multiply(A=@n22.RGB, B=NamedRerouteUsage())) {constB=2}
  n24 = TextureSampleParameter2D(UVs=NamedRerouteUsage()) {parameterName=Eye, texture=/VehicleCosmetics/Wheels/DemonEyes/Textures/T_DemonEye_Rot_BC}
  n25 = Multiply(A=@n22.RGB, B=@n24.RGB)
  n26 = Multiply(A=NamedRerouteUsage(), B=@n24.A)
  NamedRerouteDeclaration(None=LinearInterpolate(
    A=Multiply(A=@n21, B=@n22.RGB)
    B=LinearInterpolate(A=@n23, B=@n25, Alpha=@n26)
    Alpha=ScalarParameter() {defaultValue=0.8, parameterName=Lerp_Particles_Eye}
   )) {name=Emission, nodeColor=<rgba 0 0.25 1 1>}
  SetMaterialAttributes(Emissive Color=NamedRerouteUsage()) {attributeSetTypes [#1]}
 ]
}
